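# Material function that builds the layered Transylvania terrain material from world-aligned grass, dirt, gravel and cliff textures.
# Height-lerp layer blends stack dirt over grass by vertex blue, gravel by green, and cliffs by alpha, then feed SedimentGradient_TRV.
# Vertex blue also tints grass toward dead colours, gravel is desaturated and recoloured, and cliff projection coordinates are skewed between Z and XY.
# A cliff moss blend and UseCliffs switch remain unconnected to the output.
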MaterialFunction /Game/Environments/Sets/Transylvania/Terrain/MF_TRV_Terrain
graph {nodes [
  n1 = StaticBool() {value=true}
  n2 = WorldPosition()
  n3 = ComponentMask(None=@n2.XYZ) {g=true, r=true}
  n4 = ComponentMask(None=@n2.XYZ) {b=true}
  n5 = Multiply(A=ScalarParameter() {defaultValue=-0.2, group=Cliffs, parameterName=XY coordinate offset over Z}, B=@n4) {constB=0.25}
  n6 = Multiply(
   A=MaterialFunctionCall(f2=@n3) {materialFunction=&AddComponents}.f2
   B=ScalarParameter() {defaultValue=0.2, group=Cliffs, parameterName=Z coordinate offset over XY}
  ) {constB=0.25}
  n7 = MaterialFunctionCall!(
   Diffuse Texture Object=TextureObject() {texture=&/T_TRV_Cliffs_D}
   Spec Texture Object=TextureObject() {samplerType=SAMPLERTYPE_Masks, texture=&/T_TRV_Cliffs_Mask}
   TextureSize=Abs(None=ScalarParameter() {defaultValue=900, group=Cliffs, parameterName=CliffWorldCoordScale})
   --------------------------
   Use seperate settings for Z projection
   Use seperate texture for Z projection
   Texture Size Z projection
   Z Plane Diffuse Texture Object
   Z Plane Spec Texture Object
   "--------------------------      "
   ProjectionTransitionContrast=ScalarParameter() {group=Cliffs, parameterName=Cliff Projection Contrast}
   Use Texture Alpha Heightmap to Refine the Transistion=StaticBool() {value=true}
   HeightLerpTransitionContrast=ScalarParameter() {group=Cliffs, parameterName=Cliff Height Lerp Contrast}
   PickTheDiffuseHeightChannel
   TheDiffuseHeightChannel
   "--------------------------       "
   NormalTextureObject=TextureObject() {samplerType=SAMPLERTYPE_Normal, texture=&/T_TRV_Cliffs_N}
   World Space Normals=@n1
   "--------------------------  "
   Export Float 4 Diffuse=@n1
   Export Float 4 Spec
   "--------------------------     "
   WorldPosition=AppendVector(A=Add(A=@n3, B=@n5), B=Add(A=@n6, B=@n4))
   World Space Normal
  ) {materialFunction=/Engine/Functions/Engine_MaterialFunctions02/Texturing/WorldAlignedTextures_Complex}
  ComponentMask(None=@n7.XYZ Spec) {r=true}
  n8 = MaterialFunctionCall(
   Diffuse Texture Object=TextureObject() {texture=&/T_TRV_Moss_01_D}
   TextureSize=ScalarParameter() {defaultValue=250, group=Cliffs, parameterName=CliffMossScale}
   NormalTextureObject=TextureObject() {samplerType=SAMPLERTYPE_Normal, texture=&/T_TRV_Moss_01_N}
   Export Float 4 Diffuse=StaticBool() {value=true}
  ) {materialFunction=&WorldAlignedTextures_Complex}
  ComponentMask(None=@n8.XYZ Diffuse Texture) {b=true, g=true, r=true}
  n9 = TextureObjectParameter() {
   group=Dirt
   parameterName=RocksDiffuse
   samplerSource=SSM_Wrap_WorldGroupSettings
   texture=&/T_TRV_Gravel_01_D
  }
  n10 = TextureObjectParameter() {
   group=Dirt
   parameterName=RocksSpecMask
   samplerSource=SSM_Wrap_WorldGroupSettings
   samplerType=SAMPLERTYPE_Masks
   texture=&/T_TRV_Gravel_01_Mask
  }
  n11 = TextureObjectParameter() {
   group=Dirt
   parameterName=RocksNormal
   samplerSource=SSM_Wrap_WorldGroupSettings
   samplerType=SAMPLERTYPE_Normal
   texture=&/T_TRV_Gravel_01_N
  }
  n12 = MaterialFunctionCall(
   Diffuse Texture Object=@n9
   Spec Texture Object=@n10
   TextureSize=ScalarParameter() {defaultValue=900, group=Dirt, parameterName=DirtTextureTilingFrequency_1}
   NormalTextureObject=@n11
  ) {materialFunction=&WorldAlignedTextures_Complex}
  ComponentMask(None=@n12.XYZ Diffuse Texture) {a=true}
  Constant() {r=900}
  Constant() {r=2}
  Constant() {r=-2}
  Constant() {r=1000}
  Constant() {r=675}
  n13 = Power(Base=VertexColor().A, Exp=ScalarParameter() {defaultValue=0.5, parameterName=GrassMaskPower_35})
  n14 = ComponentMask(None=@n7.XYZ Diffuse Texture) {a=true}
  n15 = ScalarParameter() {defaultValue=0.1, parameterName=GrassMaskHeight}
  n16 = ScalarParameter() {defaultValue=0.7248, parameterName=GrassMaskPower_9}
  n17 = Constant() {r=1}
  n18 = MaterialFunctionCall(Float3=@n12.Z Spec) {materialFunction=&BreakOutFloat3Components}
  n19 = Multiply(A=VertexColor().G, B=Clamp(None=LinearInterpolate(A=@n16, B=@n17, Alpha=@n18.G)))
  n20 = ScalarParameter() {defaultValue=1, parameterName=GrassMaskPower_27}
  n21 = Clamp(None=Multiply(A=Power(Base=@n19, Exp=@n20), B=ScalarParameter() {defaultValue=1, parameterName=GrassMaskPower_28}))
  n22 = ScalarParameter() {defaultValue=2, parameterName=GrassMaskPower_25}
  n23 = TextureObjectParameter() {
   group=Dirt
   parameterName=DirtDiffuse
   samplerSource=SSM_Wrap_WorldGroupSettings
   texture=&/T_TRV_Dirt_D
  }
  n24 = TextureObjectParameter() {
   group=Dirt
   parameterName=DirtSpecMask
   samplerSource=SSM_Wrap_WorldGroupSettings
   samplerType=SAMPLERTYPE_Masks
   texture=&/T_TRV_Dirt_Mask
  }
  n25 = TextureObjectParameter() {
   group=Dirt
   parameterName=DirtNormal
   samplerSource=SSM_Wrap_WorldGroupSettings
   samplerType=SAMPLERTYPE_Normal
   texture=&/T_TRV_Dirt_N
  }
  n26 = MaterialFunctionCall(
   Diffuse Texture Object=@n23
   Spec Texture Object=@n24
   TextureSize=ScalarParameter() {defaultValue=675, group=Dirt, parameterName=DirtTextureTilingFrequency}
   NormalTextureObject=@n25
  ) {materialFunction=&WorldAlignedTextures_Complex}
  n27 = MaterialFunctionCall(Float3=@n26.Z Spec) {materialFunction=&BreakOutFloat3Components}
  n28 = Add(A=@n18.G, B=Multiply(A=Power(Base=@n27.R, Exp=@n22), B=ScalarParameter() {defaultValue=0.4, parameterName=GrassMaskPower_26}))
  n29 = VertexColor()
  n30 = ScalarParameter() {defaultValue=-1, parameterName=GrassMaskPower_23}
  n31 = ScalarParameter() {defaultValue=2, parameterName=GrassMaskPower_6}
  n32 = Power(Base=@n27.R, Exp=ScalarParameter() {defaultValue=2, parameterName=GrassMaskPower_2})
  n33 = TextureObjectParameter() {
   group=Grass
   parameterName=GrassDiffuse
   samplerSource=SSM_Wrap_WorldGroupSettings
   texture=&/T_TRV_Grass_01_D
  }
  n34 = TextureObjectParameter() {
   group=Grass
   parameterName=GrassNormal
   samplerSource=SSM_Wrap_WorldGroupSettings
   samplerType=SAMPLERTYPE_Normal
   texture=&/T_TRV_Grass_01_N
  }
  n35 = MaterialFunctionCall(
   Diffuse Texture Object=@n33
   TextureSize=ScalarParameter() {defaultValue=500, group=Grass, parameterName=GrassTiling}
   NormalTextureObject=@n34
   Export Float 4 Diffuse=StaticBool() {value=true}
  ) {materialFunction=&WorldAlignedTextures_Complex}
  n36 = ComponentMask(None=@n35.Z Diffuse Texture) {a=true}
  n37 = Power(Base=@n36, Exp=ScalarParameter() {defaultValue=0.0709, group=Grass, parameterName=GrassSpecPower_1})
  n38 = ComponentMask(None=@n35.Z Diffuse Texture) {g=true}
  n39 = ScalarParameter() {defaultValue=0.9928, group=Grass, parameterName=GrassDeadEdgePower_1}
  n40 = LinearInterpolate(
   A=ComponentMask(None=@n35.Z Diffuse Texture) {b=true, g=true, r=true}
   B=LinearInterpolate(
    A=VectorParameter() {parameterName=GrassDeadEdgeColor_1, defaultValue=<rgba 0.01 0.0059 0.0006 1>}.RGB
    B=VectorParameter() {parameterName=GrassDeadEdgeColor, defaultValue=<rgba 2 1.0063 0.3 1>}.RGB
    Alpha=Power(Base=@n38, Exp=@n39)
   )
   Alpha=VertexColor().B
  )
  n41 = ScalarParameter() {group=Grass, parameterName=GrassSpecAmountA}
  n42 = ScalarParameter() {defaultValue=1, group=Grass, parameterName=GrassSpecAmountB}
  n43 = ScalarParameter() {defaultValue=1, group=Grass, parameterName=GrassRoughB}
  n44 = ScalarParameter() {defaultValue=0.3, group=Grass, parameterName=GrassRoughA}
  n45 = LinearInterpolate(
   A=ScalarParameter() {defaultValue=0.0917, parameterName=GrassMaskPower_11}
   B=ScalarParameter() {defaultValue=0.3437, parameterName=GrassMaskPower_10}
   Alpha=@n27.G
  )
  n46 = ScalarParameter() {defaultValue=0.02, parameterName=SpecularFar}
  n47 = MaterialFunctionCall(
   Fade Length=ScalarParameter() {defaultValue=1534.5118, parameterName=SpecularFadeLength}
   Fade Offset=ScalarParameter() {defaultValue=500, parameterName=SpecularFadeOffset}
  ) {materialFunction=&CameraDepthFade}
  n48 = ScalarParameter() {defaultValue=0.1, parameterName=GrassMaskPower_13}
  n49 = ScalarParameter() {defaultValue=0.7, parameterName=GrassMaskPower_12}
  n50 = MaterialFunctionCall!(
   Base Material=MakeMaterialAttributes!(
    BaseColor=ComponentMask(None=@n40) {b=true, g=true, r=true}
    Metallic=Constant()
    Specular=LinearInterpolate(A=@n41, B=@n42, Alpha=@n36)
    Roughness=LinearInterpolate(A=@n43, B=@n44, Alpha=@n36)
    Anisotropy
    EmissiveColor
    Opacity=@n36
    OpacityMask
    Normal=@n35.Normal Z Texture
    Tangent
    WorldPositionOffset
    SubsurfaceColor
    ClearCoat
    ClearCoatRoughness
    AmbientOcclusion
    Refraction
    CustomizedUVs_0
    CustomizedUVs_1
    CustomizedUVs_2
    CustomizedUVs_3
    CustomizedUVs_4
    CustomizedUVs_5
    CustomizedUVs_6
    CustomizedUVs_7
    PixelDepthOffset
    ShadingModel
    Displacement
   )
   Top Material=MakeMaterialAttributes(
    BaseColor=ComponentMask(None=@n26.Z Diffuse Texture) {b=true, g=true, r=true}
    Specular=LinearInterpolate(A=@n45, B=@n46, Alpha=@n47.Result)
    Roughness=LinearInterpolate(A=@n48, B=@n49, Alpha=@n27.B)
    Opacity=@n27.R
    Normal=@n26.Normal Z Texture
   )
   Alpha=MaterialFunctionCall!(
    Diffuse 1
    Diffuse 2
    Transistion Phase=OneMinus(None=@n29.B)
    Height Texture 1=LinearInterpolate(A=@n30, B=@n31, Alpha=@n32)
    Height Texture 2=@n37
    Contrast
   ) {materialFunction=/Engine/Functions/Engine_MaterialFunctions02/Texturing/HeightLerpWithTwoHeightMaps}.Lerp Alpha No Contrast
  ) {materialFunction=/Engine/Functions/MaterialLayerFunctions/MatLayerBlend_Standard}
  n51 = ComponentMask(None=MaterialFunctionCall!(Mat attributes=@n50.Blended Material) {materialFunction=/Engine/Functions/Engine_MaterialFunctions02/MatLayerBlend_BreakOpacity}.Result) {r=true}
  n52 = Power(
   Base=Desaturation(
    None=ComponentMask(None=@n12.XYZ Diffuse Texture) {b=true, g=true, r=true}
    Fraction=ScalarParameter() {defaultValue=0.8, group=Dirt, parameterName=DirtTextureTilingFrequency_3}
   )
   Exp=ScalarParameter() {defaultValue=0.7, parameterName=GrassMaskPower_44}
  )
  n53 = Multiply(
   A=MaterialFunctionCall(
    1=VectorParameter() {parameterName=DirtBoost_2, defaultValue=<rgba 0.79 0.6926 0.5846 1>}.RGB
    2=VectorParameter() {parameterName=DirtBoost_3, defaultValue=<rgba 0.05 0.0389 0.0286 1>}.RGB
    3=VectorParameter() {parameterName=DirtBoost_4, defaultValue=<rgba 0.91 0.594 0.4083 1>}.RGB
    A=@n52
   ) {materialFunction=&Lerp_Multiple_Float3}.Lerp 3 Inputs
   B=@n52
  )
  n54 = Multiply(
   A=VectorParameter() {parameterName=DirtBoost_5, defaultValue=<rgba 0.095 0.0546 0.0393 1>}.RGB
   B=OneMinus(None=Power(Base=@n18.B) {constExponent=0.5})
  )
  n55 = ScalarParameter() {defaultValue=0.1, parameterName=GrassMaskPower_17}
  n56 = ScalarParameter() {defaultValue=0.9, parameterName=GrassMaskPower_16}
  n57 = Clamp(None=LinearInterpolate(
    A=ScalarParameter() {defaultValue=-5, parameterName=GrassMaskPower_21}
    B=ScalarParameter() {defaultValue=1, parameterName=GrassMaskPower_20}
    Alpha=@n18.R
   ))
  n58 = LinearInterpolate(
   A=ScalarParameter() {defaultValue=0.2, parameterName=GrassMaskPower_39}
   B=ScalarParameter() {defaultValue=0.8, parameterName=GrassMaskPower_38}
   Alpha=@n18.B
  )
  n59 = MaterialFunctionCall(
   Base Material=@n50.Blended Material
   Top Material=MakeMaterialAttributes(
    BaseColor=Add(A=@n53, B=@n54)
    Specular=LinearInterpolate(A=@n55, B=@n56, Alpha=@n57)
    Roughness=Clamp(None=@n58)
    Opacity=@n18.G
    Normal=@n12.Normal Z Texture
   )
   Alpha=MaterialFunctionCall(
    Transistion Phase=OneMinus(None=@n21)
    Height Texture 1=Clamp(None=@n28)
    Height Texture 2=OneMinus(None=@n51)
    Contrast=ScalarParameter() {defaultValue=0.5, parameterName=GrassMaskPower_24}
   ) {materialFunction=&HeightLerpWithTwoHeightMaps}.Alpha
  ) {materialFunction=&MatLayerBlend_Standard}
  n60 = LinearInterpolate(
   A=ScalarParameter() {parameterName=GrassMaskPower_14}
   B=ScalarParameter() {defaultValue=3, parameterName=GrassMaskPower_1}
   Alpha=ComponentMask(None=MaterialFunctionCall(Mat attributes=@n59.Blended Material) {materialFunction=&MatLayerBlend_BreakOpacity}.Result) {r=true}
  )
  n61 = ComponentMask(None=@n7.XYZ Spec) {g=true}
  n62 = ScalarParameter() {defaultValue=0.7339, parameterName=Roughness_Cliffs_1}
  n63 = ScalarParameter() {defaultValue=1, parameterName=Roughness_Cliffs_2}
  n64 = ComponentMask(None=@n7.XYZ Spec) {b=true}
  n65 = MaterialFunctionCall(
   Base Material=@n59.Blended Material
   Top Material=MakeMaterialAttributes(
    BaseColor=ComponentMask(None=@n7.XYZ Diffuse Texture) {b=true, g=true, r=true}
    Specular=@n61
    Roughness=LinearInterpolate(A=@n62, B=@n63, Alpha=@n64)
    Normal=@n7.Normal XYZ Flat Top
   )
   Alpha=MaterialFunctionCall(
    Transistion Phase=OneMinus(None=@n13)
    Height Texture 1=Multiply(A=@n14, B=@n15)
    Height Texture 2=Clamp(None=@n60)
    Contrast=ScalarParameter() {defaultValue=1, parameterName=CliffBlendContrast}
   ) {materialFunction=&HeightLerpWithTwoHeightMaps}.Alpha
  ) {materialFunction=&MatLayerBlend_Standard}
  FunctionOutput(None=MaterialFunctionCall!(In=@n65.Blended Material) {materialFunction=/Game/Environments/Sets/Arid/Materials/SedimentGradient_TRV}.Result)
  n66 = VectorParameter() {parameterName=MossFresnel, defaultValue=<rgba 0.15 0.1287 0.027 1>}
  n67 = ScalarParameter() {defaultValue=-0.7, group=Cliffs, parameterName=CliffMossNormalStrength}
  n68 = VertexNormalWS()
  n69 = Normalize(VectorInput=Add(A=@n8.Normal XYZ Texture, B=Multiply(A=@n67, B=@n68)))
  n70 = Fresnel(ExponentIn=ScalarParameter() {defaultValue=2, parameterName=MossFresnelExpo}, Normal=@n69) {baseReflectFraction=0}
  n71 = VectorParameter() {parameterName=CliffMossA, defaultValue=<rgba 0.2 0.1273 0.0167 1>}
  n72 = VectorParameter() {parameterName=CliffMossB, defaultValue=<rgba 0.03 0.018 0.0025 1>}
  n73 = ComponentMask(None=@n8.XYZ Diffuse Texture) {a=true}
  n74 = Power(Base=@n73, Exp=ScalarParameter() {defaultValue=2, parameterName=CliffDesat_4})
  n75 = LinearInterpolate(
   A=ScalarParameter() {defaultValue=-1.8, parameterName=RocksTransitionPower_5}
   B=ScalarParameter() {defaultValue=1.5, parameterName=RocksTransitionPower_6}
   Alpha=@n73
  )
  n76 = ScalarParameter() {defaultValue=1.1, parameterName=CliffDesat_5}
  n77 = Clamp(None=LinearInterpolate(
    A=ScalarParameter() {defaultValue=3, parameterName=RocksTransitionPower_1}
    B=ScalarParameter() {defaultValue=-2, parameterName=RocksTransitionPower_2}
    Alpha=Power(Base=@n64, Exp=@n76)
   ))
  n78 = VertexNormalWS()
  n79 = OneMinus(None=ConstantBiasScale(None=ComponentMask(None=@n78) {b=true}))
  n80 = ScalarParameter() {defaultValue=3, parameterName=ty754yey4}
  n81 = OneMinus(None=ConstantBiasScale(None=ComponentMask(None=@n7.Normal XYZ Flat Top) {b=true}))
  n82 = ScalarParameter() {defaultValue=2, parameterName=CliffDesat_13}
  n83 = Add(A=Clamp(None=Power(Base=@n79, Exp=@n80)) {clampMode=CMODE_ClampMax}, B=Clamp(None=Power(Base=@n81, Exp=@n82)) {clampMode=CMODE_ClampMax})
  n84 = Clamp(None=LinearInterpolate(
    A=ScalarParameter() {defaultValue=-0.3, parameterName=RocksTransitionPower_10}
    B=ScalarParameter() {defaultValue=1.5, parameterName=RocksTransitionPower_11}
    Alpha=Clamp(None=@n83)
   ))
  n85 = MaterialFunctionCall(
   Transition Phase=Multiply(A=Clamp(None=@n75), B=Multiply(A=@n77, B=@n84))
   Height Texture=Power(Base=@n73, Exp=ScalarParameter() {defaultValue=4, parameterName=CliffDesat_2})
   Contrast=ScalarParameter() {defaultValue=1, parameterName=MossBlendContrast}
  ) {materialFunction=&HeightLerp}
  LinearInterpolate(
   A=ComponentMask(None=@n7.XYZ Diffuse Texture) {b=true, g=true, r=true}
   B=Add(A=Multiply(A=@n66.RGB, B=@n70), B=LinearInterpolate(A=@n71.RGB, B=@n72.RGB, Alpha=@n74))
   Alpha=@n85.Alpha
  )
  n86 = ScalarParameter() {defaultValue=0.6, parameterName=CliffDesat_6}
  LinearInterpolate(
   A=LinearInterpolate(
    A=ScalarParameter() {defaultValue=1, parameterName=CliffDesat_10}
    B=ScalarParameter() {defaultValue=0.4, parameterName=CliffDesat_3}
    Alpha=Power(Base=@n64, Exp=@n86)
   )
   Alpha=@n85.Alpha
  )
  LinearInterpolate(
   A=@n7.Normal XYZ Flat Top
   B=@n69
   Alpha=LinearInterpolate(B=ScalarParameter() {defaultValue=0.6, parameterName=MossNormalBlend}, Alpha=@n85.Alpha)
  )
  n87 = Power(Base=@n27.R, Exp=ScalarParameter() {parameterName=GrassMaskPower_42})
  n88 = LinearInterpolate(
   A=ScalarParameter() {defaultValue=106235608, parameterName=GrassMaskPower_31}
   B=ScalarParameter() {defaultValue=12015.0049, parameterName=GrassMaskPower_15}
   Alpha=Power(Base=@n36, Exp=ScalarParameter() {defaultValue=0.5, parameterName=GrassMaskPower_8})
  )
  Multiply(A=Clamp(None=Add(A=@n87, B=@n88)), B=@n37)
  Multiply(A=@n61, B=OneMinus(None=@n85.Alpha))
  Multiply(
   A=Power(Base=Add(A=@n64, B=@n85.Alpha) {desc=Expensive}, Exp=ScalarParameter() {defaultValue=1, parameterName=GrassMaskPower_32})
   B=ScalarParameter() {defaultValue=0.2, parameterName=GrassMaskPower_33}
  )
  StaticBool() {value=true}
  StaticBool() {value=true}
  StaticSwitchParameter(True=@n65.Blended Material, False=@n65.Blended Material) {defaultValue=true, desc=Diffuse, group=Layers, parameterName=UseCliffs}
 ]}
