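# Material function for winter terrain that outputs grass, optional cliff and snow layers as blended material attributes.
# Top-projected world-aligned grass textures drive a two-color tint, specular and roughness.
# UseCliffs, off by default, height-blends a world-aligned cliff, with sheared coordinates and gray-tinted upward faces, using vertex alpha.
# UseSnow, on by default, applies MF_SharedSnow last; dirt, road-rock and debug branches are left unconnected to the output.
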MaterialFunction /Game/Environments/Sets/Winter/Terrain/MF_Winter_Terrain
graph {nodes [
  n1 = MaterialFunctionCall!(
   Diffuse Texture Object=TextureObject() {texture=/Game/Packages/DS_Fortnite_Terrain_NoLOD/Textures/T_Grasslands_D}
   Spec Texture Object
   TextureSize=ScalarParameter() {defaultValue=500, group=Grass, parameterName=GrassTiling}
   --------------------------
   Use seperate settings for Z projection
   Use seperate texture for Z projection
   Texture Size Z projection
   Z Plane Diffuse Texture Object
   Z Plane Spec Texture Object
   "--------------------------      "
   ProjectionTransitionContrast
   Use Texture Alpha Heightmap to Refine the Transistion
   HeightLerpTransitionContrast
   PickTheDiffuseHeightChannel
   TheDiffuseHeightChannel
   "--------------------------       "
   NormalTextureObject=TextureObject() {samplerType=SAMPLERTYPE_Normal, texture=/Game/Packages/DS_Fortnite_Terrain_NoLOD/Textures/T_Grasslands_N}
   World Space Normals
   "--------------------------  "
   Export Float 4 Diffuse=StaticBool() {value=true}
   Export Float 4 Spec
   "--------------------------     "
   WorldPosition
   World Space Normal
  ) {materialFunction=/Engine/Functions/Engine_MaterialFunctions02/Texturing/WorldAlignedTextures_Complex}
  n2 = ComponentMask(None=@n1.Z Diffuse Texture) {g=true}
  n3 = Power(Base=@n2, Exp=ScalarParameter() {defaultValue=1.5, group=Grass, parameterName=GrassRoughA_2})
  n4 = ScalarParameter() {defaultValue=10, group=Grass, parameterName=GrassRoughA_3}
  Clamp(None=LinearInterpolate(
    A=VectorParameter() {parameterName=GrassColor1_2, defaultValue=<rgba 0.03 0.0256 0.0214 1>}.RGB
    B=VectorParameter() {parameterName=GrassColor1_1, defaultValue=<rgba 1.5 0.9 0.3225 1>}.RGB
    Alpha=Multiply(A=@n3, B=@n4)
   )) {clampMode=CMODE_ClampMax}
  ComponentMask(None=@n1.Z Diffuse Texture) {b=true, g=true, r=true}
  n5 = VectorParameter() {parameterName=CliffWorldCoordScaleV3, defaultValue=<rgba 1000 1000 1021.7948 1>}
  n6 = StaticBool() {value=true}
  n7 = WorldPosition()
  n8 = ComponentMask(None=@n7.XYZ) {g=true, r=true}
  n9 = ComponentMask(None=@n7.XYZ) {b=true}
  n10 = Multiply(A=ScalarParameter() {defaultValue=-0.1905, group=Cliffs, parameterName=XY coordinate offset over Z}, B=@n9) {constB=0.25}
  n11 = Multiply(
   A=MaterialFunctionCall(f2=@n8) {materialFunction=&AddComponents}.f2
   B=ScalarParameter() {defaultValue=0.5048, group=Cliffs, parameterName=Z coordinate offset over XY}
  ) {constB=0.25}
  n12 = MaterialFunctionCall(
   Diffuse Texture Object=TextureObject() {texture=/Game/Packages/DS_Fortnite_Terrain_NoLOD/Textures/T_Cave_Exterior_Wall_Winter_D}
   Spec Texture Object=TextureObject() {texture=/Game/Packages/DS_Fortnite_Terrain_NoLOD/Textures/T_Cave_Exterior_Wall_S}
   TextureSize=Abs(None=ComponentMask(None=@n5.RGB) {b=true, g=true, r=true})
   ProjectionTransitionContrast=ScalarParameter() {defaultValue=1, group=Cliffs, parameterName=Cliff Projection Contrast}
   Use Texture Alpha Heightmap to Refine the Transistion=StaticBool() {value=true}
   HeightLerpTransitionContrast=ScalarParameter() {defaultValue=1, group=Cliffs, parameterName=Cliff Height Lerp Contrast}
   NormalTextureObject=TextureObject() {samplerType=SAMPLERTYPE_Normal, texture=/Game/Packages/DS_Fortnite_Terrain_NoLOD/Textures/T_Cave_Exterior_Wall_N}
   World Space Normals=@n6
   Export Float 4 Diffuse=@n6
   WorldPosition=AppendVector(A=Add(A=@n8, B=@n10), B=Add(A=@n11, B=@n9))
  ) {materialFunction=&WorldAlignedTextures_Complex}
  ComponentMask(None=@n12.XYZ Spec) {g=true}
  Constant() {r=900}
  Constant() {r=1000}
  n13 = VertexColor()
  n14 = ScalarParameter() {defaultValue=0.5, parameterName=GrassMaskPower_35}
  n15 = ComponentMask(None=@n12.XYZ Diffuse Texture) {a=true}
  n16 = ScalarParameter() {parameterName=GrassMaskPower_14}
  n17 = ScalarParameter() {defaultValue=3, parameterName=GrassMaskPower_1}
  n18 = ComponentMask(None=@n1.Z Diffuse Texture) {a=true}
  n19 = Add(A=@n2, B=Multiply(A=ScalarParameter() {defaultValue=0.5, group=Grass, parameterName=GrassRoughA_1}, B=@n18))
  n20 = ScalarParameter() {defaultValue=0.7, group=Grass, parameterName=GrassRoughA_4}
  n21 = MakeMaterialAttributes!(
   BaseColor=LinearInterpolate(
    A=VectorParameter() {parameterName=GrassColor1_3, defaultValue=<rgba 0 0 0 1>}.RGB
    B=VectorParameter() {parameterName=GrassColor1, defaultValue=<rgba 0.465 0.2868 0.1379 1>}.RGB
    Alpha=Power(Base=@n19, Exp=@n20) {constExponent=0.25}
   )
   Metallic=Constant()
   Specular=LinearInterpolate(
    A=ScalarParameter() {group=Grass, parameterName=GrassSpecAmountA}
    B=ScalarParameter() {defaultValue=0.2, group=Grass, parameterName=GrassSpecAmountB}
    Alpha=@n18
   )
   Roughness=LinearInterpolate(
    A=ScalarParameter() {defaultValue=0.8, group=Grass, parameterName=GrassRoughB}
    B=ScalarParameter() {defaultValue=0.1, group=Grass, parameterName=GrassRoughA}
    Alpha=@n18
   )
   Anisotropy
   EmissiveColor
   Opacity=@n18
   OpacityMask
   Normal=@n1.Normal Z Texture
   Tangent
   WorldPositionOffset
   SubsurfaceColor
   ClearCoat
   ClearCoatRoughness
   AmbientOcclusion
   Refraction
   CustomizedUVs_0
   CustomizedUVs_1
   CustomizedUVs_2
   CustomizedUVs_3
   CustomizedUVs_4
   CustomizedUVs_5
   CustomizedUVs_6
   CustomizedUVs_7
   PixelDepthOffset
   ShadingModel
   Displacement
  )
  n22 = ComponentMask(None=MaterialFunctionCall!(Mat attributes=@n21) {materialFunction=/Engine/Functions/Engine_MaterialFunctions02/MatLayerBlend_BreakOpacity}.Result) {r=true}
  n23 = MaterialFunctionCall!(
   Diffuse 1
   Diffuse 2
   Transistion Phase=OneMinus(None=Power(Base=@n13.A, Exp=@n14))
   Height Texture 1=Multiply(A=@n15, B=ScalarParameter() {defaultValue=0.1, parameterName=GrassMaskHeight})
   Height Texture 2=Clamp(None=LinearInterpolate(A=@n16, B=@n17, Alpha=@n22))
   Contrast=ScalarParameter() {defaultValue=1, parameterName=CliffBlendContrast}
  ) {materialFunction=/Engine/Functions/Engine_MaterialFunctions02/Texturing/HeightLerpWithTwoHeightMaps}
  n24 = VectorParameter() {group=Cliffs, parameterName=CliffsDiffuseTint, defaultValue=<rgba 0.9896 0.9465 0.9005 1>}
  n25 = ComponentMask(None=@n12.XYZ Diffuse Texture) {b=true, g=true, r=true}
  n26 = ComponentMask(None=@n12.Normal XYZ Flat Top) {b=true}
  n27 = Multiply(
   A=Power(Base=OneMinus(None=@n15), Exp=ScalarParameter() {defaultValue=1, group=Cliffs, parameterName=CliffsDiffusePower_2})
   B=MaterialFunctionCall(In=ConstantBiasScale(None=@n26), Contrast=ScalarParameter() {defaultValue=3, parameterName=SnowNormalMask02}) {materialFunction=&CheapContrast}.Result
  )
  n28 = LinearInterpolate(
   A=Power(Base=Multiply(A=@n24.RGB, B=@n25), Exp=ScalarParameter() {defaultValue=1.5, group=Cliffs, parameterName=CliffsDiffusePower})
   B=VectorParameter() {group=Cliffs, parameterName=CliffsDiffuseTint_1, defaultValue=<rgba 0.276 0.276 0.276 1>}.RGB
   Alpha=@n27
  )
  n29 = ComponentMask(None=@n12.XYZ Spec) {r=true}
  n30 = Add(
   A=@n27
   B=LinearInterpolate(
    A=ScalarParameter() {defaultValue=0.1, group=Cliffs, parameterName=Roughness_CliffsA}
    B=ScalarParameter() {defaultValue=1, group=Cliffs, parameterName=Roughness_CliffsB}
    Alpha=ComponentMask(None=@n12.XYZ Spec) {b=true}
   )
  )
  n31 = ScalarParameter() {defaultValue=-0.1, group=Cliffs, parameterName=Opacity_CliffsA}
  n32 = ScalarParameter() {defaultValue=1, group=Cliffs, parameterName=Opacity_CliffsB}
  n33 = Clamp(None=Add(A=Multiply(A=@n29) {constB=2}, B=LinearInterpolate(A=@n31, B=@n32, Alpha=@n15)))
  n34 = StaticSwitchParameter(
   True=MaterialFunctionCall!(
    Base Material=@n21
    Top Material=MakeMaterialAttributes(BaseColor=@n28, Specular=@n29, Roughness=@n30, Opacity=@n33, Normal=@n12.Normal XYZ Flat Top)
    Alpha=@n23.Alpha
   ) {materialFunction=/Engine/Functions/MaterialLayerFunctions/MatLayerBlend_Standard}.Blended Material
   False=@n21
  ) {desc=Diffuse, group=Layers, parameterName=UseCliffs}
  n35 = MaterialFunctionCall!(MaterialLayers=@n34, UseVertexBlend) {materialFunction=&/MF_SharedSnow}
  FunctionOutput(None=StaticSwitchParameter(True=@n35.OutputBlendedMaterial, False=@n34) {defaultValue=true, desc=Diffuse, group=Layers, parameterName=UseSnow})
  LinearInterpolate(
   A=ScalarParameter() {group=Grass, parameterName=GrassSpecAmountA_1}
   B=ScalarParameter() {defaultValue=1, group=Grass, parameterName=GrassSpecAmountB_1}
  )
  n36 = BreakMaterialAttributes!(Attr=@n35.OutputBlendedMaterial)
  MakeMaterialAttributes(
   BaseColor=Multiply(A=ScalarParameter() {parameterName=5wedggg}, B=@n36.BaseColor)
   Specular=@n36.Specular
   Roughness=@n36.Specular
   EmissiveColor=Multiply(A=@n35.OutputZMask, B=VectorParameter() {parameterName=debugdebuggg, defaultValue=<rgba 0.46 0 0 1>}.RGB)
   Normal=@n36.Specular
  )
  n37 = ScalarParameter() {parameterName=GrassMaskPower_23}
  n38 = ScalarParameter() {defaultValue=1, parameterName=GrassMaskPower_6}
  n39 = TextureObjectParameter() {group=Dirt, parameterName=DirtDiffuse, samplerSource=SSM_Wrap_WorldGroupSettings, texture=/Game/Packages/DS_Fortnight/WhiteRockCustom_D}
  n40 = TextureObjectParameter() {
   group=Dirt
   parameterName=DirtNormal
   samplerSource=SSM_Wrap_WorldGroupSettings
   samplerType=SAMPLERTYPE_Normal
   texture=/Game/Packages/DS_Fortnight/WhiteRockCustom_N
  }
  n41 = MaterialFunctionCall(
   Diffuse Texture Object=@n39
   TextureSize=ScalarParameter() {defaultValue=690, group=Dirt, parameterName=DirtTextureTilingFrequency}
   NormalTextureObject=@n40
   Export Float 4 Diffuse=StaticBool() {value=true}
  ) {materialFunction=&WorldAlignedTextures_Complex}
  n42 = ComponentMask(None=@n41.Z Diffuse Texture) {a=true}
  n43 = Power(Base=@n42, Exp=ScalarParameter() {defaultValue=2, parameterName=GrassMaskPower_2})
  n44 = Power(Base=@n18, Exp=ScalarParameter() {defaultValue=1, group=Grass, parameterName=GrassSpecPower_1})
  n45 = MaterialFunctionCall(
   Transistion Phase=OneMinus(None=VertexColor().B)
   Height Texture 1=Clamp(None=LinearInterpolate(A=@n37, B=@n38, Alpha=@n43))
   Height Texture 2=@n44
  ) {materialFunction=&HeightLerpWithTwoHeightMaps}
  n46 = Multiply(
   A=VectorParameter() {parameterName=DirtModColor, defaultValue=<rgba 0.385 0.2695 0.1944 1>}.RGB
   B=Desaturation(
    None=ComponentMask(None=@n41.Z Diffuse Texture) {b=true, g=true, r=true}
    Fraction=ScalarParameter() {group=Grass, parameterName=DirtSaturation}
   )
  )
  n47 = ComponentMask(None=@n41.Z Diffuse Texture) {r=true}
  n48 = LinearInterpolate(A=ScalarParameter() {parameterName=GrassMaskPower_11}, B=ScalarParameter() {defaultValue=0.1, parameterName=GrassMaskPower_10}, Alpha=@n47)
  n49 = ScalarParameter() {defaultValue=0.02, parameterName=SpecularFar}
  n50 = MaterialFunctionCall(
   Fade Length=ScalarParameter() {defaultValue=1534.5118, parameterName=SpecularFadeLength}
   Fade Offset=ScalarParameter() {defaultValue=500, parameterName=SpecularFadeOffset}
  ) {materialFunction=&CameraDepthFade}
  n51 = ScalarParameter() {defaultValue=0.8, parameterName=GrassMaskPower_13}
  n52 = ScalarParameter() {defaultValue=0.1, parameterName=GrassMaskPower_12}
  n53 = Add(
   A=@n41.Normal Z Texture
   B=Multiply(A=ScalarParameter() {defaultValue=-0.7, group=Cliffs, parameterName=CliffMossNormalStrength_1}, B=VertexNormalWS())
  )
  MaterialFunctionCall(
   Base Material=@n21
   Top Material=MakeMaterialAttributes(
    BaseColor=Clamp(None=@n46) {clampMode=CMODE_ClampMax}
    Specular=LinearInterpolate(A=@n48, B=@n49, Alpha=@n50.Result)
    Roughness=LinearInterpolate(A=@n51, B=@n52, Alpha=@n47)
    Opacity=@n42
    Normal=Normalize(VectorInput=@n53)
   )
   Alpha=@n45.Lerp Alpha No Contrast
  ) {materialFunction=&MatLayerBlend_Standard}
  n54 = ScalarParameter() {defaultValue=0.7248, parameterName=GrassMaskPower_9}
  n55 = Constant() {r=1}
  n56 = TextureObjectParameter() {
   group=Dirt
   parameterName=RocksDiffuse
   samplerSource=SSM_Wrap_WorldGroupSettings
   texture=/Game/Packages/DS_Fortnight/SM/Textures/T_Street_Road_D
  }
  n57 = TextureObjectParameter() {
   group=Dirt
   parameterName=RocksSpecMask
   samplerSource=SSM_Wrap_WorldGroupSettings
   samplerType=SAMPLERTYPE_Masks
   texture=/Game/Packages/DS_Fortnight/SM/Textures/T_Street_Road_M
  }
  n58 = TextureObjectParameter() {
   group=Dirt
   parameterName=RocksNormal
   samplerSource=SSM_Wrap_WorldGroupSettings
   samplerType=SAMPLERTYPE_Normal
   texture=/Game/Packages/DS_Fortnight/SM/Textures/T_Street_Road_N
  }
  n59 = MaterialFunctionCall(
   Diffuse Texture Object=@n56
   Spec Texture Object=@n57
   TextureSize=ScalarParameter() {defaultValue=768, group=Dirt, parameterName=DirtTextureTilingFrequency_1}
   NormalTextureObject=@n58
   Export Float 4 Diffuse=StaticBool() {value=true}
  ) {materialFunction=&WorldAlignedTextures_Complex}
  n60 = ComponentMask(None=@n59.Z Diffuse Texture) {a=true}
  n61 = Multiply(A=VertexColor().R, B=Clamp(None=LinearInterpolate(A=@n54, B=@n55, Alpha=@n60)))
  n62 = ScalarParameter() {defaultValue=1, parameterName=GrassMaskPower_27}
  n63 = Clamp(None=Multiply(A=Power(Base=@n61, Exp=@n62), B=ScalarParameter() {defaultValue=1, parameterName=GrassMaskPower_28}))
  n64 = ScalarParameter() {defaultValue=2, parameterName=GrassMaskPower_25}
  n65 = Add(A=@n60, B=Multiply(A=Power(Base=@n42, Exp=@n64), B=ScalarParameter() {defaultValue=0.4, parameterName=GrassMaskPower_26}))
  n66 = ComponentMask(None=MaterialFunctionCall(Mat attributes=@n21) {materialFunction=&MatLayerBlend_BreakOpacity}.Result) {r=true}
  n67 = ScalarParameter() {defaultValue=0.1, parameterName=GrassMaskPower_17}
  n68 = ScalarParameter() {defaultValue=0.9, parameterName=GrassMaskPower_16}
  n69 = MaterialFunctionCall(Float3=@n59.Z Spec) {materialFunction=&BreakOutFloat3Components}
  n70 = Clamp(None=LinearInterpolate(
    A=ScalarParameter() {defaultValue=-5, parameterName=GrassMaskPower_21}
    B=ScalarParameter() {defaultValue=1, parameterName=GrassMaskPower_20}
    Alpha=@n69.R
   ))
  n71 = LinearInterpolate(
   A=ScalarParameter() {defaultValue=0.2, parameterName=GrassMaskPower_39}
   B=ScalarParameter() {defaultValue=0.8, parameterName=GrassMaskPower_38}
   Alpha=@n69.B
  )
  MaterialFunctionCall(
   Base Material=@n21
   Top Material=MakeMaterialAttributes(
    BaseColor=ComponentMask(None=@n59.Z Diffuse Texture) {b=true, g=true, r=true}
    Specular=LinearInterpolate(A=@n67, B=@n68, Alpha=@n70)
    Roughness=Clamp(None=@n71)
    Opacity=@n60
    Normal=@n59.Normal Z Texture
   )
   Alpha=MaterialFunctionCall(
    Transistion Phase=OneMinus(None=@n63)
    Height Texture 1=Clamp(None=@n65)
    Height Texture 2=OneMinus(None=@n66)
    Contrast=ScalarParameter() {defaultValue=0.5, parameterName=GrassMaskPower_24}
   ) {materialFunction=&HeightLerpWithTwoHeightMaps}.Alpha
  ) {materialFunction=&MatLayerBlend_Standard}
  MaterialFunctionCall(Float3=@n41.Z Spec) {materialFunction=&BreakOutFloat3Components}
  Multiply(A=@n45.Lerp Alpha No Contrast)
  Multiply(A=@n23.Alpha)
  Multiply()
  n72 = Power(Base=@n42, Exp=ScalarParameter() {defaultValue=1, parameterName=GrassMaskPower_42})
  n73 = LinearInterpolate(
   A=ScalarParameter() {defaultValue=-1, parameterName=GrassMaskPower_31}
   B=ScalarParameter() {defaultValue=1, parameterName=GrassMaskPower_15}
   Alpha=Power(Base=@n18, Exp=ScalarParameter() {defaultValue=0.5, parameterName=GrassMaskPower_8})
  )
  Multiply(A=Clamp(None=Add(A=@n72, B=@n73)), B=@n44)
  n74 = ScalarParameter() {group=Cliffs, parameterName=CliffMossNormalStrength_3}
  n75 = VertexNormalWS()
  Normalize(VectorInput=Add(A=@n1.Normal Z Texture, B=Multiply(A=@n74, B=@n75)))
  Reroute(None=@n41.Normal Z Texture)
  ScalarParameter() {defaultValue=400, group=Cliffs, parameterName=CliffWorldCoordScale}
  TextureObjectParameter() {
   group=Dirt
   parameterName=DirtSpecMask
   samplerSource=SSM_Wrap_WorldGroupSettings
   samplerType=SAMPLERTYPE_Masks
   texture=/Game/Environments/Props/SinkHole/Textures/T_SinkHole_Mud_01_S
  }
  TextureObjectParameter() {
   group=Grass
   parameterName=GrassDiffuse
   samplerSource=SSM_Wrap_WorldGroupSettings
   texture=/Game/Environments/Sets/Transylvania/Terrain/T_TRV_Grass_01_D
  }
  TextureObjectParameter() {
   group=Grass
   parameterName=GrassNormal
   samplerSource=SSM_Wrap_WorldGroupSettings
   samplerType=SAMPLERTYPE_Normal
   texture=/Game/Environments/Sets/Transylvania/Terrain/T_TRV_Grass_01_N
  }
  TextureObject() {samplerType=SAMPLERTYPE_Normal, texture=/Game/Environments/Sets/Transylvania/Terrain/T_TRV_Gravel_01_N}
  TextureObject() {samplerType=SAMPLERTYPE_Masks, texture=/Game/Environments/Sets/Transylvania/Terrain/T_TRV_Gravel_01_Mask}
  TextureObject() {texture=/Game/Environments/Sets/Transylvania/Terrain/T_TRV_Gravel_01_D}
  TextureObject() {texture=/Game/Environments/Props/SinkHole/Textures/T_SinkHole_Mud_01_D}
  TextureObject() {samplerType=SAMPLERTYPE_Masks, texture=&T_SinkHole_Mud_01_S}
  TextureObject() {samplerType=SAMPLERTYPE_Normal, texture=/Game/Environments/Props/SinkHole/Textures/T_SinkHole_Mud_01_N}
  TextureObject() {texture=&T_TRV_Grass_01_D}
  TextureObject() {samplerType=SAMPLERTYPE_Normal, texture=&T_TRV_Grass_01_N}
 ]}
