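# Nanite-enabled Transylvania terrain master material, using world-space normals and material attributes.
# Static switches select the MF_TRV_Terrain function, the Winter terrain function or a plain red fallback.
# An inline graph also height-blends world-aligned grass, dirt, gravel and cliff layers in sequence, driven by vertex-colour B, G and A.
# Side branches, not fed into that blend, compute cliff moss, a sediment gradient and vertex-colour debug views.
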Material /Game/Environments/Sets/Transylvania/Terrain/M_TRV_Terrain
graph {
 properties {
  bTangentSpaceNormal=false
  bUseMaterialAttributes=true
  bUsedWithNanite=true
  pixelDepthOffsetMode=PDOM_Legacy
  usedShadingModels=Default Lit
 }
 nodes [
  n1 = StaticBool() {value=true}
  n2 = WorldPosition()
  n3 = ComponentMask(None=@n2.XYZ) {g=true, r=true}
  n4 = ComponentMask(None=@n2.XYZ) {b=true}
  n5 = Multiply(A=ScalarParameter() {defaultValue=-0.2, group=Cliffs, parameterName=XY coordinate offset over Z}, B=@n4) {constB=0.25}
  n6 = Multiply(
   A=MaterialFunctionCall(f2=@n3) {materialFunction=&AddComponents}.f2
   B=ScalarParameter() {defaultValue=0.2, group=Cliffs, parameterName=Z coordinate offset over XY}
  ) {constB=0.25}
  n7 = MaterialFunctionCall!(
   Diffuse Texture Object=TextureObject() {texture=&/T_TRV_Cliffs_D}
   Spec Texture Object=TextureObject() {samplerType=SAMPLERTYPE_Masks, texture=&/T_TRV_Cliffs_Mask}
   TextureSize=Abs(None=ScalarParameter() {defaultValue=900, group=Cliffs, parameterName=CliffWorldCoordScale})
   --------------------------
   Use seperate settings for Z projection
   Use seperate texture for Z projection
   Texture Size Z projection
   Z Plane Diffuse Texture Object
   Z Plane Spec Texture Object
   "--------------------------      "
   ProjectionTransitionContrast=ScalarParameter() {group=Cliffs, parameterName=Cliff Projection Contrast}
   Use Texture Alpha Heightmap to Refine the Transistion=StaticBool() {value=true}
   HeightLerpTransitionContrast=ScalarParameter() {group=Cliffs, parameterName=Cliff Height Lerp Contrast}
   PickTheDiffuseHeightChannel
   TheDiffuseHeightChannel
   "--------------------------       "
   NormalTextureObject=TextureObject() {samplerType=SAMPLERTYPE_Normal, texture=&/T_TRV_Cliffs_N}
   World Space Normals=@n1
   "--------------------------  "
   Export Float 4 Diffuse=@n1
   Export Float 4 Spec
   "--------------------------     "
   WorldPosition=AppendVector(A=Add(A=@n3, B=@n5), B=Add(A=@n6, B=@n4))
   World Space Normal
  ) {materialFunction=/Engine/Functions/Engine_MaterialFunctions02/Texturing/WorldAlignedTextures_Complex}
  ComponentMask(None=@n7.XYZ Spec) {r=true}
  n8 = MaterialFunctionCall(
   Diffuse Texture Object=TextureObject() {texture=&/T_TRV_Moss_01_D}
   TextureSize=ScalarParameter() {defaultValue=250, group=Cliffs, parameterName=CliffMossScale}
   NormalTextureObject=TextureObject() {samplerType=SAMPLERTYPE_Normal, texture=&/T_TRV_Moss_01_N}
   Export Float 4 Diffuse=StaticBool() {value=true}
  ) {materialFunction=&WorldAlignedTextures_Complex}
  ComponentMask(None=@n8.XYZ Diffuse Texture) {b=true, g=true, r=true}
  n9 = TextureObjectParameter() {
   group=Dirt
   parameterName=RocksDiffuse
   samplerSource=SSM_Wrap_WorldGroupSettings
   texture=&/T_TRV_Gravel_01_D
  }
  n10 = TextureObjectParameter() {
   group=Dirt
   parameterName=RocksSpecMask
   samplerSource=SSM_Wrap_WorldGroupSettings
   samplerType=SAMPLERTYPE_Masks
   texture=&/T_TRV_Gravel_01_Mask
  }
  n11 = TextureObjectParameter() {
   group=Dirt
   parameterName=RocksNormal
   samplerSource=SSM_Wrap_WorldGroupSettings
   samplerType=SAMPLERTYPE_Normal
   texture=&/T_TRV_Gravel_01_N
  }
  n12 = MaterialFunctionCall(
   Diffuse Texture Object=@n9
   Spec Texture Object=@n10
   TextureSize=ScalarParameter() {defaultValue=900, group=Dirt, parameterName=DirtTextureTilingFrequency_1}
   NormalTextureObject=@n11
  ) {materialFunction=&WorldAlignedTextures_Complex}
  ComponentMask(None=@n12.XYZ Diffuse Texture) {a=true}
  Constant() {r=900}
  Constant() {r=2}
  Constant() {r=-2}
  Constant() {r=1000}
  Constant() {r=675}
  n13 = VectorParameter() {parameterName=MossFresnel, defaultValue=<rgba 0.15 0.1287 0.027 1>}
  n14 = ScalarParameter() {defaultValue=-0.7, group=Cliffs, parameterName=CliffMossNormalStrength}
  n15 = VertexNormalWS()
  n16 = Normalize(VectorInput=Add(A=@n8.Normal XYZ Texture, B=Multiply(A=@n14, B=@n15)))
  n17 = Fresnel(ExponentIn=ScalarParameter() {defaultValue=2, parameterName=MossFresnelExpo}, Normal=@n16) {baseReflectFraction=0}
  n18 = VectorParameter() {parameterName=CliffMossA, defaultValue=<rgba 0.2 0.1273 0.0167 1>}
  n19 = VectorParameter() {parameterName=CliffMossB, defaultValue=<rgba 0.03 0.018 0.0025 1>}
  n20 = ComponentMask(None=@n8.XYZ Diffuse Texture) {a=true}
  n21 = Power(Base=@n20, Exp=ScalarParameter() {defaultValue=2, parameterName=CliffDesat_4})
  n22 = LinearInterpolate(
   A=ScalarParameter() {defaultValue=-1.8, parameterName=RocksTransitionPower_5}
   B=ScalarParameter() {defaultValue=1.5, parameterName=RocksTransitionPower_6}
   Alpha=@n20
  )
  n23 = ScalarParameter() {defaultValue=1.1, parameterName=CliffDesat_5}
  n24 = ComponentMask(None=@n7.XYZ Spec) {b=true}
  n25 = Clamp(None=LinearInterpolate(
    A=ScalarParameter() {defaultValue=3, parameterName=RocksTransitionPower_1}
    B=ScalarParameter() {defaultValue=-2, parameterName=RocksTransitionPower_2}
    Alpha=Power(Base=@n24, Exp=@n23)
   ))
  n26 = VertexNormalWS()
  n27 = OneMinus(None=ConstantBiasScale(None=ComponentMask(None=@n26) {b=true}))
  n28 = ScalarParameter() {defaultValue=3, parameterName=ty754yey4}
  n29 = OneMinus(None=ConstantBiasScale(None=ComponentMask(None=@n7.Normal XYZ Flat Top) {b=true}))
  n30 = ScalarParameter() {defaultValue=2, parameterName=CliffDesat_13}
  n31 = Add(A=Clamp(None=Power(Base=@n27, Exp=@n28)) {clampMode=CMODE_ClampMax}, B=Clamp(None=Power(Base=@n29, Exp=@n30)) {clampMode=CMODE_ClampMax})
  n32 = Clamp(None=LinearInterpolate(
    A=ScalarParameter() {defaultValue=-0.3, parameterName=RocksTransitionPower_10}
    B=ScalarParameter() {defaultValue=1.5, parameterName=RocksTransitionPower_11}
    Alpha=Clamp(None=@n31)
   ))
  n33 = MaterialFunctionCall(
   Transition Phase=Multiply(A=Clamp(None=@n22), B=Multiply(A=@n25, B=@n32))
   Height Texture=Power(Base=@n20, Exp=ScalarParameter() {defaultValue=4, parameterName=CliffDesat_2})
   Contrast=ScalarParameter() {defaultValue=1, parameterName=MossBlendContrast}
  ) {materialFunction=&HeightLerp}
  LinearInterpolate(
   A=ComponentMask(None=@n7.XYZ Diffuse Texture) {b=true, g=true, r=true}
   B=Add(A=Multiply(A=@n13.RGB, B=@n17), B=LinearInterpolate(A=@n18.RGB, B=@n19.RGB, Alpha=@n21))
   Alpha=@n33.Alpha
  )
  n34 = ScalarParameter() {defaultValue=0.6, parameterName=CliffDesat_6}
  LinearInterpolate(
   A=LinearInterpolate(
    A=ScalarParameter() {defaultValue=1, parameterName=CliffDesat_10}
    B=ScalarParameter() {defaultValue=0.4, parameterName=CliffDesat_3}
    Alpha=Power(Base=@n24, Exp=@n34)
   )
   Alpha=@n33.Alpha
  )
  LinearInterpolate(
   A=@n7.Normal XYZ Flat Top
   B=@n16
   Alpha=LinearInterpolate(B=ScalarParameter() {defaultValue=0.6, parameterName=MossNormalBlend}, Alpha=@n33.Alpha)
  )
  n35 = VertexColor()
  n36 = Add(
   A=Multiply(A=@n35.R, B=VectorParameter() {parameterName=Red, defaultValue=<rgba 1 0 0 1>}.RGB)
   B=Multiply(A=@n35.G, B=VectorParameter() {parameterName=Green, defaultValue=<rgba 0 1 0 1>}.RGB)
  )
  n37 = Multiply(A=@n35.B, B=VectorParameter() {parameterName=Blue, defaultValue=<rgba 0 0 1 1>}.RGB)
  n38 = ScalarParameter() {parameterName=sdfy45t34_1}
  LinearInterpolate(B=Multiply(A=Add(A=@n36, B=@n37), B=@n38), Alpha=@n38) {constB=0}
  n39 = LinearInterpolate(
   A=LinearInterpolate(
    A=Constant3Vector() {constant=<rgba 0.0073 0.045 0.0027 1>}
    B=Constant3Vector() {constant=<rgba 0.135 0.0934 0.0614 1>}
    Alpha=VertexColor().B
   )
   B=Constant3Vector() {constant=<rgba 0.25 0.2446 0.2387 1>}
   Alpha=VertexColor().G
  )
  n40 = Constant3Vector() {constant=<rgba 0.48 0.4276 0.0768 1>}
  n41 = VertexColor()
  n42 = Constant3Vector() {constant=<rgba 0.1803 0.1318 0.31 1>}
  n43 = VertexColor()
  LinearInterpolate(
   B=StaticSwitchParameter(True=LinearInterpolate(A=@n39, B=@n40, Alpha=@n41.A), False=LinearInterpolate(A=@n39, B=@n42, Alpha=@n43.R)) {defaultValue=true, desc=Diffuse, group=Layers, parameterName=UseCliffs}
   Alpha=ScalarParameter() {defaultValue=1, parameterName=DEBUG-VertColor}
  ) {constB=0}
  n44 = Power(Base=VertexColor().A, Exp=ScalarParameter() {defaultValue=0.5, parameterName=GrassMaskPower_35})
  n45 = ComponentMask(None=@n7.XYZ Diffuse Texture) {a=true}
  n46 = ScalarParameter() {defaultValue=0.1, parameterName=GrassMaskHeight}
  n47 = ScalarParameter() {defaultValue=0.7248, parameterName=GrassMaskPower_9}
  n48 = Constant() {r=1}
  n49 = MaterialFunctionCall(Float3=@n12.Z Spec) {materialFunction=&BreakOutFloat3Components}
  n50 = Multiply(A=VertexColor().G, B=Clamp(None=LinearInterpolate(A=@n47, B=@n48, Alpha=@n49.G)))
  n51 = ScalarParameter() {defaultValue=1, parameterName=GrassMaskPower_27}
  n52 = Clamp(None=Multiply(A=Power(Base=@n50, Exp=@n51), B=ScalarParameter() {defaultValue=1, parameterName=GrassMaskPower_28}))
  n53 = ScalarParameter() {defaultValue=2, parameterName=GrassMaskPower_25}
  n54 = TextureObjectParameter() {
   group=Dirt
   parameterName=DirtDiffuse
   samplerSource=SSM_Wrap_WorldGroupSettings
   texture=&/T_TRV_Dirt_D
  }
  n55 = TextureObjectParameter() {
   group=Dirt
   parameterName=DirtSpecMask
   samplerSource=SSM_Wrap_WorldGroupSettings
   samplerType=SAMPLERTYPE_Masks
   texture=&/T_TRV_Dirt_Mask
  }
  n56 = TextureObjectParameter() {
   group=Dirt
   parameterName=DirtNormal
   samplerSource=SSM_Wrap_WorldGroupSettings
   samplerType=SAMPLERTYPE_Normal
   texture=&/T_TRV_Dirt_N
  }
  n57 = MaterialFunctionCall(
   Diffuse Texture Object=@n54
   Spec Texture Object=@n55
   TextureSize=ScalarParameter() {defaultValue=675, group=Dirt, parameterName=DirtTextureTilingFrequency}
   NormalTextureObject=@n56
  ) {materialFunction=&WorldAlignedTextures_Complex}
  n58 = MaterialFunctionCall(Float3=@n57.Z Spec) {materialFunction=&BreakOutFloat3Components}
  n59 = Add(A=@n49.G, B=Multiply(A=Power(Base=@n58.R, Exp=@n53), B=ScalarParameter() {defaultValue=0.4, parameterName=GrassMaskPower_26}))
  n60 = VertexColor()
  n61 = ScalarParameter() {defaultValue=-1, parameterName=GrassMaskPower_23}
  n62 = ScalarParameter() {defaultValue=2, parameterName=GrassMaskPower_6}
  n63 = Power(Base=@n58.R, Exp=ScalarParameter() {defaultValue=2, parameterName=GrassMaskPower_2})
  n64 = TextureObjectParameter() {
   group=Grass
   parameterName=GrassDiffuse
   samplerSource=SSM_Wrap_WorldGroupSettings
   texture=&/T_TRV_Grass_01_D
  }
  n65 = TextureObjectParameter() {
   group=Grass
   parameterName=GrassNormal
   samplerSource=SSM_Wrap_WorldGroupSettings
   samplerType=SAMPLERTYPE_Normal
   texture=&/T_TRV_Grass_01_N
  }
  n66 = MaterialFunctionCall(
   Diffuse Texture Object=@n64
   TextureSize=ScalarParameter() {defaultValue=500, group=Grass, parameterName=GrassTiling}
   NormalTextureObject=@n65
   Export Float 4 Diffuse=StaticBool() {value=true}
  ) {materialFunction=&WorldAlignedTextures_Complex}
  n67 = ComponentMask(None=@n66.Z Diffuse Texture) {a=true}
  n68 = Power(Base=@n67, Exp=ScalarParameter() {defaultValue=0.0709, group=Grass, parameterName=GrassSpecPower_1})
  n69 = ComponentMask(None=@n66.Z Diffuse Texture) {g=true}
  n70 = ScalarParameter() {defaultValue=0.9928, group=Grass, parameterName=GrassDeadEdgePower_1}
  n71 = LinearInterpolate(
   A=ComponentMask(None=@n66.Z Diffuse Texture) {b=true, g=true, r=true}
   B=LinearInterpolate(
    A=VectorParameter() {parameterName=GrassDeadEdgeColor_1, defaultValue=<rgba 0.01 0.0059 0.0006 1>}.RGB
    B=VectorParameter() {parameterName=GrassDeadEdgeColor, defaultValue=<rgba 2 1.0063 0.3 1>}.RGB
    Alpha=Power(Base=@n69, Exp=@n70)
   )
   Alpha=VertexColor().B
  )
  n72 = ScalarParameter() {group=Grass, parameterName=GrassSpecAmountA}
  n73 = ScalarParameter() {defaultValue=1, group=Grass, parameterName=GrassSpecAmountB}
  n74 = ScalarParameter() {defaultValue=1, group=Grass, parameterName=GrassRoughB}
  n75 = ScalarParameter() {defaultValue=0.3, group=Grass, parameterName=GrassRoughA}
  n76 = LinearInterpolate(
   A=ScalarParameter() {defaultValue=0.0917, parameterName=GrassMaskPower_11}
   B=ScalarParameter() {defaultValue=0.3437, parameterName=GrassMaskPower_10}
   Alpha=@n58.G
  )
  n77 = ScalarParameter() {defaultValue=0.02, parameterName=SpecularFar}
  n78 = MaterialFunctionCall(
   Fade Length=ScalarParameter() {defaultValue=1534.5118, parameterName=SpecularFadeLength}
   Fade Offset=ScalarParameter() {defaultValue=500, parameterName=SpecularFadeOffset}
  ) {materialFunction=&CameraDepthFade}
  n79 = ScalarParameter() {defaultValue=0.1, parameterName=GrassMaskPower_13}
  n80 = ScalarParameter() {defaultValue=0.7, parameterName=GrassMaskPower_12}
  n81 = MaterialFunctionCall!(
   Base Material=MakeMaterialAttributes!(
    BaseColor=ComponentMask(None=@n71) {b=true, g=true, r=true}
    Metallic=Constant()
    Specular=LinearInterpolate(A=@n72, B=@n73, Alpha=@n67)
    Roughness=LinearInterpolate(A=@n74, B=@n75, Alpha=@n67)
    Anisotropy
    EmissiveColor
    Opacity=@n67
    OpacityMask
    Normal=@n66.Normal Z Texture
    Tangent
    WorldPositionOffset
    SubsurfaceColor
    ClearCoat
    ClearCoatRoughness
    AmbientOcclusion
    Refraction
    CustomizedUVs_0
    CustomizedUVs_1
    CustomizedUVs_2
    CustomizedUVs_3
    CustomizedUVs_4
    CustomizedUVs_5
    CustomizedUVs_6
    CustomizedUVs_7
    PixelDepthOffset
    ShadingModel
    Displacement
   )
   Top Material=MakeMaterialAttributes(
    BaseColor=ComponentMask(None=@n57.Z Diffuse Texture) {b=true, g=true, r=true}
    Specular=LinearInterpolate(A=@n76, B=@n77, Alpha=@n78.Result)
    Roughness=LinearInterpolate(A=@n79, B=@n80, Alpha=@n58.B)
    Opacity=@n58.R
    Normal=@n57.Normal Z Texture
   )
   Alpha=MaterialFunctionCall!(
    Diffuse 1
    Diffuse 2
    Transistion Phase=OneMinus(None=@n60.B)
    Height Texture 1=LinearInterpolate(A=@n61, B=@n62, Alpha=@n63)
    Height Texture 2=@n68
    Contrast
   ) {materialFunction=/Engine/Functions/Engine_MaterialFunctions02/Texturing/HeightLerpWithTwoHeightMaps}.Lerp Alpha No Contrast
  ) {materialFunction=/Engine/Functions/MaterialLayerFunctions/MatLayerBlend_Standard}
  n82 = ComponentMask(None=MaterialFunctionCall!(Mat attributes=@n81.Blended Material) {materialFunction=/Engine/Functions/Engine_MaterialFunctions02/MatLayerBlend_BreakOpacity}.Result) {r=true}
  n83 = VectorParameter() {parameterName=DirtBoost_2, defaultValue=<rgba 0.79 0.6926 0.5846 1>}
  n84 = VectorParameter() {parameterName=DirtBoost_3, defaultValue=<rgba 0.05 0.0389 0.0286 1>}
  n85 = VectorParameter() {parameterName=DirtBoost_4, defaultValue=<rgba 0.91 0.594 0.4083 1>}
  n86 = Power(
   Base=Desaturation(
    None=ComponentMask(None=@n12.XYZ Diffuse Texture) {b=true, g=true, r=true}
    Fraction=ScalarParameter() {defaultValue=0.8, group=Dirt, parameterName=DirtTextureTilingFrequency_3}
   )
   Exp=ScalarParameter() {defaultValue=0.7, parameterName=GrassMaskPower_44}
  )
  n87 = Power(Base=@n49.B) {constExponent=0.5}
  n88 = Add(
   A=Multiply(
    A=MaterialFunctionCall(1=@n83.RGB, 2=@n84.RGB, 3=@n85.RGB, A=@n86) {materialFunction=&Lerp_Multiple_Float3}.Lerp 3 Inputs
    B=@n86
   )
   B=Multiply(A=VectorParameter() {parameterName=DirtBoost_5, defaultValue=<rgba 0.095 0.0546 0.0393 1>}.RGB, B=OneMinus(None=@n87))
  )
  n89 = ScalarParameter() {defaultValue=0.1, parameterName=GrassMaskPower_17}
  n90 = ScalarParameter() {defaultValue=0.9, parameterName=GrassMaskPower_16}
  n91 = Clamp(None=LinearInterpolate(
    A=ScalarParameter() {defaultValue=-5, parameterName=GrassMaskPower_21}
    B=ScalarParameter() {defaultValue=1, parameterName=GrassMaskPower_20}
    Alpha=@n49.R
   ))
  n92 = LinearInterpolate(
   A=ScalarParameter() {defaultValue=0.2, parameterName=GrassMaskPower_39}
   B=ScalarParameter() {defaultValue=0.8, parameterName=GrassMaskPower_38}
   Alpha=@n49.B
  )
  n93 = MaterialFunctionCall(
   Base Material=@n81.Blended Material
   Top Material=MakeMaterialAttributes(
    BaseColor=ComponentMask(None=@n88) {b=true, g=true, r=true}
    Specular=LinearInterpolate(A=@n89, B=@n90, Alpha=@n91)
    Roughness=Clamp(None=@n92)
    Opacity=@n49.G
    Normal=@n12.Normal Z Texture
   )
   Alpha=MaterialFunctionCall(
    Transistion Phase=OneMinus(None=@n52)
    Height Texture 1=Clamp(None=@n59)
    Height Texture 2=OneMinus(None=@n82)
    Contrast=ScalarParameter() {defaultValue=0.5, parameterName=GrassMaskPower_24}
   ) {materialFunction=&HeightLerpWithTwoHeightMaps}.Alpha
  ) {materialFunction=&MatLayerBlend_Standard}
  n94 = LinearInterpolate(
   A=ScalarParameter() {parameterName=GrassMaskPower_14}
   B=ScalarParameter() {defaultValue=3, parameterName=GrassMaskPower_1}
   Alpha=ComponentMask(None=MaterialFunctionCall(Mat attributes=@n93.Blended Material) {materialFunction=&MatLayerBlend_BreakOpacity}.Result) {r=true}
  )
  n95 = ComponentMask(None=@n7.XYZ Spec) {g=true}
  n96 = ScalarParameter() {defaultValue=0.7339, parameterName=Roughness_Cliffs_1}
  n97 = ScalarParameter() {defaultValue=1, parameterName=Roughness_Cliffs_2}
  n98 = MaterialFunctionCall(
   Base Material=@n93.Blended Material
   Top Material=MakeMaterialAttributes(
    BaseColor=ComponentMask(None=@n7.XYZ Diffuse Texture) {b=true, g=true, r=true}
    Specular=@n95
    Roughness=LinearInterpolate(A=@n96, B=@n97, Alpha=@n24)
    Normal=@n7.Normal XYZ Flat Top
   )
   Alpha=MaterialFunctionCall(
    Transistion Phase=OneMinus(None=@n44)
    Height Texture 1=Multiply(A=@n45, B=@n46)
    Height Texture 2=Clamp(None=@n94)
    Contrast=ScalarParameter() {defaultValue=1, parameterName=CliffBlendContrast}
   ) {materialFunction=&HeightLerpWithTwoHeightMaps}.Alpha
  ) {materialFunction=&MatLayerBlend_Standard}
  MaterialFunctionCall!(In=@n98.Blended Material) {materialFunction=/Game/Environments/Sets/Arid/Materials/SedimentGradient_TRV}
  n99 = Power(Base=@n58.R, Exp=ScalarParameter() {parameterName=GrassMaskPower_42})
  n100 = LinearInterpolate(
   A=ScalarParameter() {defaultValue=106235608, parameterName=GrassMaskPower_31}
   B=ScalarParameter() {defaultValue=12015.0049, parameterName=GrassMaskPower_15}
   Alpha=Power(Base=@n67, Exp=ScalarParameter() {defaultValue=0.5, parameterName=GrassMaskPower_8})
  )
  Multiply(A=Clamp(None=Add(A=@n99, B=@n100)), B=@n68)
  Multiply(A=@n95, B=OneMinus(None=@n33.Alpha))
  Multiply(
   A=Power(Base=Add(A=@n24, B=@n33.Alpha) {desc=Expensive}, Exp=ScalarParameter() {defaultValue=1, parameterName=GrassMaskPower_32})
   B=ScalarParameter() {defaultValue=0.2, parameterName=GrassMaskPower_33}
  )
  StaticBool() {value=true}
  StaticBool() {value=true}
  StaticSwitchParameter(True=LinearInterpolate(A=Constant3Vector() {constant=<rgba 1 0.5 0 0>}, B=Constant3Vector() {constant=<rgba 0 1 0.5 0>}, Alpha=@n35.A)) {defaultValue=true, desc=Diffuse, group=Layers, parameterName=UseCliffs}
  n101 = Constant3Vector() {constant=<rgba 1 0 0 0>}
  n102 = Constant()
  n103 = Constant() {r=1}
  n104 = StaticSwitch(
   True=MaterialFunctionCall() {materialFunction=&/MF_TRV_Terrain}.Result
   False=StaticSwitch(
    True=MaterialFunctionCall() {materialFunction=/Game/Environments/Sets/Winter/Terrain/MF_Winter_Terrain}.Result
    False=MakeMaterialAttributes(BaseColor=@n101, Metallic=@n102, Specular=@n102, Roughness=@n103)
    Value=StaticBoolParameter() {group=01_MasterMaterialSettings, parameterName=Use_WinterTerrain}
   )
   Value=StaticBoolParameter() {defaultValue=true, group=01_MasterMaterialSettings, parameterName=Use_TransylvaniaTerrain}
  )
  StaticSwitchParameter(True=@n104, False=@n104) {defaultValue=true, group=Layers, parameterName="Use Grass(true) or a BaseTilingTexture(False)"}
  StaticSwitchParameter(True=@n98.Blended Material, False=@n98.Blended Material) {defaultValue=true, desc=Diffuse, group=Layers, parameterName=UseCliffs}
 ]
}
